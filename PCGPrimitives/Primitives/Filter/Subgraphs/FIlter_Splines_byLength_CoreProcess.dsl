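; PCG subgraph that sorts splines and 2D polygons by length, using the Length and Operator parameters, and outputs both groups tagged Selected or Unselected.
; Splines are filtered directly on their data-domain Length attribute.
; Polygons become points, keep only HoleIndex -1, form closed linear splines for the same test, then are rebuilt as polygons, so holes are dropped.
; Any existing Selected or Unselected tags are cleared before the new tag is applied.
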(graph FIlter_Splines_byLength_CoreProcess
  (params (Length :type Double :default 1e+01) (Operator :type (Enum /Script/PCG.EPCGAttributeFilterOperator)))
  (bind FilterDataByType_0 (Filter%20Data%20By%20Type :TargetType (struct :Ids [(struct :Struct /Script/PCG.PCGDataTypeInfoSpline)])))
  (bind FilterDataByAttribute_1 (Filter%20Data%20By%20Attribute :FilterMode FilterByValue :TargetAttribute @data.$Length :Threshold (struct :bUseConstantThreshold false) :Filter Length :FilterOperator-param Operator))
  (bind FilterDataByType_1 (Filter%20Data%20By%20Type :TargetType (struct :Ids [(struct :Struct /Script/PCG.PCGDataTypeInfoPolygon2D)])))
  (bind ToPoint_0 (To%20Point))
  (bind AttributeFilter_0 (Filter%20Attribute%20Elements :AttributeTypes (struct :Type Integer32 :Int32Value -1) :bUseConstantThreshold true :Operator Equal :TargetAttribute HoleIndex :ThresholdAttribute @Last))
  (bind CreateSpline_1 (Create%20Spline :bClosedLoop true :bLinear true :InterpTypeAttribute None))
  (bind FilterDataByAttribute_2 (Filter%20Data%20By%20Attribute :FilterMode FilterByValue :TargetAttribute @data.$Length :Threshold (struct :bUseConstantThreshold false) :Filter Length :FilterOperator-param Operator))
  (bind CreatePolygon2D_0 (Create%20Polygon%202D :InputType ForceClosed))
  (bind DeleteTags_1 (Delete%20Tags :SelectedTags "Selected, Unselected"))
  (bind AddTags_1 (Add%20Tags :TagsToAdd "Selected"))
  (bind CreatePolygon2D_1 (Create%20Polygon%202D :InputType ForceClosed))
  (bind DeleteTags_2 (Delete%20Tags :SelectedTags "Selected, Unselected"))
  (bind AddTags_2 (Add%20Tags :TagsToAdd "Unselected"))
  (connect DeleteTags_1 Out AddTags_1 In)
  (connect DeleteTags_2 Out AddTags_2 In)
  (connect ToPoint_0 Out AttributeFilter_0 In)
  (connect FilterDataByAttribute_2 InsideFilter CreatePolygon2D_0 In)
  (connect FilterDataByAttribute_2 OutsideFilter CreatePolygon2D_1 In)
  (connect AttributeFilter_0 InsideFilter CreateSpline_1 In)
  (connect CreatePolygon2D_0 Out DeleteTags_1 In)
  (connect FilterDataByAttribute_1 InsideFilter DeleteTags_1 In)
  (connect CreatePolygon2D_1 Out DeleteTags_2 In)
  (connect FilterDataByAttribute_1 OutsideFilter DeleteTags_2 In)
  (connect FilterDataByType_0 InsideFilter FilterDataByAttribute_1 In)
  (connect CreateSpline_1 Out FilterDataByAttribute_2 In)
  (connect DefaultInputNode In FilterDataByType_0 In)
  (connect DefaultInputNode In FilterDataByType_1 In)
  (connect FilterDataByType_1 InsideFilter ToPoint_0 In)
  (output Out AddTags_1)
  (output Out AddTags_2))
