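; PCG subgraph that sorts area data into selected and unselected sets by size.
; It keeps data carrying an area attribute, then compares that area to the Size parameter, scaled from square metres to square centimetres, using the Operator parameter.
; Each side has old selection tags cleared and is retagged Selected or Unselected.
; Both sides feed the same output; the Discard_unselected parameter is not used here.
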(graph Filter_Areas_BySize_CoreProcess
  (params (Size :type Double :default 1e+05) (Operator :type (Enum /Script/PCG.EPCGAttributeFilterOperator)) (Discard_unselected :type Bool :default true))
  (bind AttributeMathsOp_0 (Attribute%20Maths%20Op :Operation Multiply :InA Size)) ; M square to cm square
  (bind FilterDataByAttribute_0 (Filter%20Data%20By%20Attribute :Attribute @data.$area))
  (bind FilterDataByAttribute_3 (Filter%20Data%20By%20Attribute :FilterMode FilterByValue :TargetAttribute @data.$area :Threshold (struct :bUseConstantThreshold false) :FilterOperator-param Operator))
  (bind DeleteTags_0 (Delete%20Tags :SelectedTags "Selected, Unselected"))
  (bind AddTags_0 (Add%20Tags :TagsToAdd "Selected"))
  (bind DeleteTags_1 (Delete%20Tags :SelectedTags "Selected, Unselected"))
  (bind AddTags_1 (Add%20Tags :TagsToAdd "Unselected"))
  (connect DeleteTags_0 Out AddTags_0 In)
  (connect DeleteTags_1 Out AddTags_1 In)
  (connect FilterDataByAttribute_3 InsideFilter DeleteTags_0 In)
  (connect FilterDataByAttribute_3 OutsideFilter DeleteTags_1 In)
  (connect DefaultInputNode In FilterDataByAttribute_0 In)
  (connect AttributeMathsOp_0 Out FilterDataByAttribute_3 Filter)
  (connect FilterDataByAttribute_0 InsideFilter FilterDataByAttribute_3 In)
  (output Out AddTags_0)
  (output Out AddTags_1))
