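; PCG subgraph that finds the smallest-area data among its inputs and tags it Selected, with the rest tagged Unselected.
; Inputs carrying a data-level area attribute have that value converted to points, merged and reduced to a minimum.
; A second filter splits those inputs by equality to this minimum, and old selection tags are removed.
; Invert_Filter drives two Branch nodes that swap which side receives each tag; both sets go to Out.
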(graph Filter_Areas_Smallest_CoreProcess
  (params (Invert_Filter :type Bool))
  (bind FilterDataByAttribute_0 (Filter%20Data%20By%20Attribute :Attribute @data.$area))
  (bind CopyAttributes_0 (Copy%20Attributes :InputSource @data.$area :OutputTarget area))
  (bind ExtractAttribute_0 (Extract%20Attribute%20at%20Index :bForceOutputAttributeToBeInElementsDomain true :InputSource @data.$area))
  (bind AttributeSetToPoint_0 (Attribute%20Set%20To%20Point))
  (bind MergePoints_0 (Merge%20Points))
  (bind AttributeReduce_0 (Attribute%20Reduce :InputSource area :Operation Min))
  (bind FilterDataByAttribute_1 (Filter%20Data%20By%20Attribute :FilterMode FilterByValue :FilterOperator Equal :TargetAttribute @data.$area :Threshold (struct :bUseConstantThreshold false)))
  (bind DeleteTags_0 (Delete%20Tags :SelectedTags "Selected, Unselected"))
  (bind DeleteTags_1 (Delete%20Tags :SelectedTags "Selected, Unselected"))
  (bind Branch_1 (Branch :bOutputToB-param Invert_Filter))
  (bind Branch_2 (Branch :bOutputToB-param Invert_Filter))
  (bind AddTags_2 (Add%20Tags :TagsToAdd "Selected"))
  (bind AddTags_3 (Add%20Tags :TagsToAdd "Unselected"))
  (connect Branch_1 Output%20B AddTags_2 In)
  (connect Branch_2 Output%20A AddTags_2 In)
  (connect Branch_1 Output%20A AddTags_3 In)
  (connect Branch_2 Output%20B AddTags_3 In)
  (connect MergePoints_0 Out AttributeReduce_0 In)
  (connect ExtractAttribute_0 Out AttributeSetToPoint_0 In)
  (connect DeleteTags_1 Out Branch_1 In)
  (connect DeleteTags_0 Out Branch_2 In)
  (connect FilterDataByAttribute_0 InsideFilter CopyAttributes_0 Source)
  (connect FilterDataByAttribute_0 InsideFilter CopyAttributes_0 Target)
  (connect FilterDataByAttribute_1 InsideFilter DeleteTags_0 In)
  (connect FilterDataByAttribute_1 OutsideFilter DeleteTags_1 In)
  (connect CopyAttributes_0 Out ExtractAttribute_0 In)
  (connect DefaultInputNode In FilterDataByAttribute_0 In)
  (connect AttributeReduce_0 Out FilterDataByAttribute_1 Filter)
  (connect FilterDataByAttribute_0 InsideFilter FilterDataByAttribute_1 In)
  (connect AttributeSetToPoint_0 Out MergePoints_0 In)
  (output Out AddTags_2)
  (output Out AddTags_3))
