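; PCG core subgraph for the area filter primitive that filters Polygon2D areas by their containment within other areas.
; It strips attributes and Selected/Unselected tags, then separates data tagged Input:0 from the rest, keeping only Polygon2D on each side.
; The Input:0 areas feed an inner-loop subgraph, with the others as clipping polygons and a flag to include partially contained areas.
; The loop's result is the graph output.
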(graph Filter_Areas_Within_Areas_CoreProcess
  (params (include_partially_contained_areas :type Bool))
  (bind DeleteAttributes_0 (Delete%20Attributes :Operation KeepSelectedAttributes))
  (bind DeleteTags_0 (Delete%20Tags :SelectedTags "Selected, Unselected"))
  (bind FilterDataByTag_0 (Filter%20Data%20By%20Tag :SelectedTags "Input:0"))
  (bind FilterDataByType_0 (Filter%20Data%20By%20Type :TargetType (struct :Ids [(struct :Struct /Script/PCG.PCGDataTypeInfoPolygon2D)])))
  (bind FilterDataByType_4 (Filter%20Data%20By%20Type :TargetType (struct :Ids [(struct :Struct /Script/PCG.PCGDataTypeInfoPolygon2D)])))
  (bind Loop_0 (Loop :Graph /PCGPrimitives/Primitives/Filter/Subgraphs/Filter_Areas_Within_Areas_CoreProcess_innerLoop.Filter_Areas_Within_Areas_CoreProcess_innerLoop :include_partially_contained_areas include_partially_contained_areas))
  (connect DefaultInputNode In DeleteAttributes_0 In)
  (connect DeleteAttributes_0 Out DeleteTags_0 In)
  (connect DeleteTags_0 Out FilterDataByTag_0 In)
  (connect FilterDataByTag_0 InsideFilter FilterDataByType_0 In)
  (connect FilterDataByTag_0 OutsideFilter FilterDataByType_4 In)
  (connect FilterDataByType_4 InsideFilter Loop_0 Clipping%20polygons)
  (connect FilterDataByType_0 InsideFilter Loop_0 In)
  (output Out Loop_0))
