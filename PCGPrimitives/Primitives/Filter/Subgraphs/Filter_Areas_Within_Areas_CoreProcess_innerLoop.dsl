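; PCG loop-body subgraph for filtering areas within areas, tagging input areas "selected" or "unselected" against clipping polygons.
; A polygon operation intersects the input with the clip polygons; an empty result means unselected.
; Otherwise, floored areas before and after clipping are compared, so an unchanged area counts as fully contained and selected.
; Partially contained areas follow include_partially_contained_areas, and the original, unclipped data is output.
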(graph Filter_Areas_Within_Areas_CoreProcess_innerLoop
  (params (include_partially_contained_areas :type Bool))
  (inputs (Clipping%20polygons :type Polygon2D))
  (bind Polygon2DOperation_0 (Polygon%20Operation))
  (bind GetElementsCount_0 (Get%20Element%20Count))
  (bind Branch_1 (Branch)) ; IF no data go unselected
  (bind AddTags_1 (Add%20Tags :TagsToAdd "unselected"))
  ;[0.15 0.15 0.15 0.5] COMPARING AREA after intersection
  (bind AttributeMathsOp_0 (Attribute%20Maths%20Op :InputSource1 @data.$area :Operation Floor :OutputTarget @data.flooredArea))
  (bind AttributeMathsOp_1 (Attribute%20Maths%20Op :InputSource1 @data.$area :Operation Floor :OutputTarget @data.flooredArea))
  (bind AttributeCompareOp_0 (Attribute%20Compare%20Op :InputSource1 @data.flooredArea :InputSource2 @data.flooredArea :OutputTarget @data.untouched))
  (bind ExtractAttribute_0 (Extract%20Attribute%20at%20Index :bForceOutputAttributeToBeInElementsDomain true :InputSource @data.untouched))
  (bind Branch_0 (Branch))
  (bind Branch_2 (Branch :bOutputToB-param include_partially_contained_areas))
  (bind AddTags_3 (Add%20Tags :TagsToAdd "selected"))
  (bind AddTags_5 (Add%20Tags :TagsToAdd "selected"))
  (bind AddTags_0 (Add%20Tags :TagsToAdd "unselected"))
  (reroute Input%20Areas In)
  (connect Branch_2 Output%20A AddTags_0 In)
  (connect Branch_1 Output%20A AddTags_1 In)
  (connect Branch_2 Output%20B AddTags_3 In)
  (connect Branch_0 Output%20B AddTags_5 In)
  (connect AttributeMathsOp_0 Out AttributeCompareOp_0 InA)
  (connect AttributeMathsOp_1 Out AttributeCompareOp_0 InB)
  (connect Polygon2DOperation_0 Out AttributeMathsOp_0 In)
  (connect Input%20Areas Out AttributeMathsOp_1 In)
  (connect ExtractAttribute_0 Out Branch_0 bOutputToB)
  (connect Branch_1 Output%20B Branch_0 In)
  (connect GetElementsCount_0 Out Branch_1 bOutputToB)
  (connect Input%20Areas Out Branch_1 In)
  (connect Branch_0 Output%20A Branch_2 In)
  (connect AttributeCompareOp_0 Out ExtractAttribute_0 In)
  (connect Polygon2DOperation_0 Out GetElementsCount_0 In)
  (connect DefaultInputNode Clipping%20polygons Polygon2DOperation_0 ClipPolygons)
  (connect Input%20Areas Out Polygon2DOperation_0 In)
  (output Out AddTags_0)
  (output Out AddTags_1)
  (output Out AddTags_3)
  (output Out AddTags_5))
