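; PCG core-process subgraph that filters point, spline and polygon data by an attribute against a typed value, tagging results Selected or Unselected.
; It takes Input:0 data carrying the attribute, then branches on a substring test of the attribute name.
; One branch filters points or control points, rebuilding splines and polygons.
; The other tests each data's first element, using temporary index tags to select and re-sort whole data.
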(graph Filter_AttributeValue_CoreProcess
  (params (Operator :type (Enum /Script/PCG.EPCGAttributeFilterOperator)) (TargetAttribute :type String :default "$Density") (FilterValue :type (Struct /Script/PCG.PCGMetadataTypesConstantStruct) :default (struct :Type Float)))
  (bind ExtractAttribute_0 (Extract%20Attribute%20at%20Index :bForceOutputAttributeToBeInElementsDomain true :InputSource Type :In FilterValue))
  (bind ExtractAttribute_1 (Extract%20Attribute%20at%20Index :bForceOutputAttributeToBeInElementsDomain true :InputSource FloatValue :In FilterValue))
  (bind ExtractAttribute_2 (Extract%20Attribute%20at%20Index :bForceOutputAttributeToBeInElementsDomain true :InputSource DoubleValue :In FilterValue))
  (bind ExtractAttribute_3 (Extract%20Attribute%20at%20Index :bForceOutputAttributeToBeInElementsDomain true :InputSource Int32Value :In FilterValue))
  (bind ExtractAttribute_4 (Extract%20Attribute%20at%20Index :bForceOutputAttributeToBeInElementsDomain true :InputSource IntValue :In FilterValue))
  (bind ExtractAttribute_5 (Extract%20Attribute%20at%20Index :bForceOutputAttributeToBeInElementsDomain true :InputSource Vector2Value :In FilterValue))
  (bind ExtractAttribute_6 (Extract%20Attribute%20at%20Index :bForceOutputAttributeToBeInElementsDomain true :InputSource VectorValue :In FilterValue))
  (bind ExtractAttribute_7 (Extract%20Attribute%20at%20Index :bForceOutputAttributeToBeInElementsDomain true :InputSource Vector4Value :In FilterValue))
  (bind ExtractAttribute_8 (Extract%20Attribute%20at%20Index :bForceOutputAttributeToBeInElementsDomain true :InputSource QuatValue :In FilterValue))
  (bind ExtractAttribute_9 (Extract%20Attribute%20at%20Index :bForceOutputAttributeToBeInElementsDomain true :InputSource TransformValue :In FilterValue))
  (bind ExtractAttribute_10 (Extract%20Attribute%20at%20Index :bForceOutputAttributeToBeInElementsDomain true :InputSource StringValue :In FilterValue))
  (bind ExtractAttribute_11 (Extract%20Attribute%20at%20Index :bForceOutputAttributeToBeInElementsDomain true :InputSource BoolValue :In FilterValue))
  (bind ExtractAttribute_12 (Extract%20Attribute%20at%20Index :bForceOutputAttributeToBeInElementsDomain true :InputSource RotatorValue :In FilterValue))
  (bind ExtractAttribute_13 (Extract%20Attribute%20at%20Index :bForceOutputAttributeToBeInElementsDomain true :InputSource NameValue :In FilterValue))
  (bind ExtractAttribute_14 (Extract%20Attribute%20at%20Index :bForceOutputAttributeToBeInElementsDomain true :InputSource SoftObjectPathValue :In FilterValue))
  (bind ExtractAttribute_15 (Extract%20Attribute%20at%20Index :bForceOutputAttributeToBeInElementsDomain true :InputSource SoftClassPathValue :In FilterValue))
  (bind Select%20%28Multi%29_0 (Select%20%28Multi%29 :EnumSelection (struct :Class /Script/PCG.EPCGMetadataTypes) :SelectionMode Enum))
  (bind FilterDataByType_4 (Filter%20Data%20By%20Type :TargetType (struct :Ids [(struct :Struct /Script/PCG.PCGDataTypeInfoPolygon2D)])))
  (bind ToPoint_1 (To%20Point))
  (bind AttributeFilter_2 (Filter%20Attribute%20Elements :bGenerateOutputDataEvenIfEmpty false :bWarnOnDataMissingAttribute false :TargetAttribute @Last :ThresholdAttribute @Last :Operator-param Operator :TargetAttribute-param TargetAttribute))
  (bind CreatePolygon2D_0 (Create%20Polygon%202D))
  (bind FilterDataByType_6 (Filter%20Data%20By%20Type :TargetType (struct :Ids [(struct :Struct /Script/PCG.PCGDataTypeInfoSpline)])))
  (bind FilterDataByAttribute_2 (Filter%20Data%20By%20Attribute :FilterMode FilterByValue :FilterOperator Equal :TargetAttribute @Data.$isClosed :Threshold (struct :AttributeTypes (struct :Type Boolean :BoolValue true))))
  (bind GetSplineControlPoints_2 (Get%20Spline%20Control%20Points))
  (bind AttributeFilter_3 (Filter%20Attribute%20Elements :bGenerateOutputDataEvenIfEmpty false :bWarnOnDataMissingAttribute false :TargetAttribute @Last :ThresholdAttribute @Last :Operator-param Operator :TargetAttribute-param TargetAttribute))
  (bind CreateSpline_0 (Create%20Spline :bApplyCustomTangents true :bClosedLoop true :bUseInterpTypeAttribute true))
  (bind GetSplineControlPoints_3 (Get%20Spline%20Control%20Points))
  (bind AttributeFilter_4 (Filter%20Attribute%20Elements :bGenerateOutputDataEvenIfEmpty false :bWarnOnDataMissingAttribute false :TargetAttribute @Last :ThresholdAttribute @Last :Operator-param Operator :TargetAttribute-param TargetAttribute))
  (bind FilterDataByAttribute_0 (Filter%20Data%20By%20Attribute :FilterMode FilterByValue :FilterOperator Equal :TargetAttribute @Data.$NumElements :Threshold (struct :AttributeTypes (struct :Type Integer32 :Int32Value 1))))
  (bind CreateSpline_2 (Create%20Spline :bApplyCustomTangents true :bUseInterpTypeAttribute true))
  (bind DeleteTags_2 (Delete%20Tags :Operator Matches :SelectedTags "TMP_DataIndex*, Selected, Unselected"))
  (bind AddTags_0 (Add%20Tags :TagsToAdd "Selected"))
  (bind CreatePolygon2D_1 (Create%20Polygon%202D))
  (bind CreateSpline_1 (Create%20Spline :bApplyCustomTangents true :bClosedLoop true :bUseInterpTypeAttribute true))
  (bind FilterDataByAttribute_3 (Filter%20Data%20By%20Attribute :FilterMode FilterByValue :FilterOperator Equal :TargetAttribute @Data.$NumElements :Threshold (struct :AttributeTypes (struct :Type Integer32 :Int32Value 1))))
  (bind CreateSpline_3 (Create%20Spline :bApplyCustomTangents true :bUseInterpTypeAttribute true))
  (bind DeleteTags_3 (Delete%20Tags :Operator Matches :SelectedTags "TMP_DataIndex*, Selected, Unselected"))
  (bind AddTags_1 (Add%20Tags :TagsToAdd "Unselected"))
  ;[0 0 0 0.5 12] Primary/Secondary Inputs
  (bind FilterDataByTag_0 (Filter%20Data%20By%20Tag :SelectedTags "Input:0"))
  ;[0 0 0 0.5] Core Process
  (bind AttributeStringOp_1 (Attribute%20String%20Op :Operation Substring :SearchCase IgnoreCase :InA TargetAttribute))
  (bind FilterDataByAttribute_1 (Filter%20Data%20By%20Attribute :Attribute-param TargetAttribute))
  (bind WriteDataIndex_0 (Write%20Data%20Index :IndexTag "TMP_DataIndex"))
  (bind Branch_0 (Branch))
  (bind FilterDataByType_5 (Filter%20Data%20By%20Type :TargetType (struct :Ids [(struct :Struct /Script/PCG.PCGDataTypeInfoPoint)])))
  (bind AttributeFilter_0 (Filter%20Attribute%20Elements :bGenerateOutputDataEvenIfEmpty false :bWarnOnDataMissingAttribute false :TargetAttribute @Last :ThresholdAttribute @Last :Operator-param Operator :TargetAttribute-param TargetAttribute))
  (bind FilterDataByType_1 (Filter%20Data%20By%20Type :TargetType (struct :Ids [(struct :Struct /Script/PCG.PCGDataTypeInfoPoint)])))
  (bind FilterDataByType_2 (Filter%20Data%20By%20Type :TargetType (struct :Ids [(struct :Struct /Script/PCG.PCGDataTypeInfoSpline)])))
  (bind GetSplineControlPoints_0 (Get%20Spline%20Control%20Points))
  (bind FilterDataByType_3 (Filter%20Data%20By%20Type :TargetType (struct :Ids [(struct :Struct /Script/PCG.PCGDataTypeInfoPolygon2D)])))
  (bind ToPoint_0 (To%20Point))
  (bind FilterElementsByIndex_0 (Filter%20Elements%20By%20Index :bSelectIndicesByInput false :SelectedIndices "0"))
  (bind AttributeFilter_1 (Filter%20Attribute%20Elements :bGenerateOutputDataEvenIfEmpty false :bWarnOnDataMissingAttribute false :TargetAttribute @Last :ThresholdAttribute @Last :Operator-param Operator :TargetAttribute-param TargetAttribute))
  (bind DeleteTags_0 (Delete%20Tags :Operation KeepOnlySelectedTags :Operator Substring :SelectedTags "TMP_DataIndex:"))
  (bind GetTags_0 (Get%20Tags :bExtractTagValues true))
  (bind MergeAttributes_0 (Merge%20Attributes :MergeTagsMode None))
  (bind AttributeReduce_0 (Attribute%20Reduce :InputSource Values :Operation Join))
  (bind DataNum_1 (Data%20Count))
  (bind Branch_1 (Branch))
  (bind FilterDataByIndex_0 (Filter%20Data%20By%20Index))
  (bind SortTags_0 (Sort%20Data%20By%20Tag%20Value :Tag TMP_DataIndex))
  (bind SortTags_1 (Sort%20Data%20By%20Tag%20Value :Tag TMP_DataIndex))
  (connect DeleteTags_2 Out AddTags_0 In)
  (connect DeleteTags_3 Out AddTags_1 In)
  (connect Select%20%28Multi%29_0 Out AttributeFilter_0 Filter)
  (connect FilterDataByType_5 InsideFilter AttributeFilter_0 In)
  (connect Select%20%28Multi%29_0 Out AttributeFilter_1 Filter)
  (connect FilterElementsByIndex_0 In%20Filter AttributeFilter_1 In)
  (connect Select%20%28Multi%29_0 Out AttributeFilter_2 Filter)
  (connect ToPoint_1 Out AttributeFilter_2 In)
  (connect Select%20%28Multi%29_0 Out AttributeFilter_3 Filter)
  (connect GetSplineControlPoints_2 Out AttributeFilter_3 In)
  (connect Select%20%28Multi%29_0 Out AttributeFilter_4 Filter)
  (connect GetSplineControlPoints_3 Out AttributeFilter_4 In)
  (connect MergeAttributes_0 Out AttributeReduce_0 In)
  (connect AttributeStringOp_1 Out Branch_0 bOutputToB)
  (connect WriteDataIndex_0 Out Branch_0 In)
  (connect DataNum_1 Out Branch_1 bOutputToB)
  (connect Branch_0 Output%20B Branch_1 In)
  (connect AttributeFilter_2 InsideFilter CreatePolygon2D_0 In)
  (connect AttributeFilter_2 OutsideFilter CreatePolygon2D_1 In)
  (connect AttributeFilter_3 InsideFilter CreateSpline_0 In)
  (connect AttributeFilter_3 OutsideFilter CreateSpline_1 In)
  (connect FilterDataByAttribute_0 OutsideFilter CreateSpline_2 In)
  (connect FilterDataByAttribute_3 OutsideFilter CreateSpline_3 In)
  (connect AttributeReduce_0 Out DataNum_1 In)
  (connect AttributeFilter_1 InsideFilter DeleteTags_0 In)
  (connect AttributeFilter_0 InsideFilter DeleteTags_2 In)
  (connect CreatePolygon2D_0 Out DeleteTags_2 In)
  (connect CreateSpline_0 Out DeleteTags_2 In)
  (connect CreateSpline_2 Out DeleteTags_2 In)
  (connect FilterDataByAttribute_0 InsideFilter DeleteTags_2 In)
  (connect SortTags_0 Out DeleteTags_2 In)
  (connect AttributeFilter_0 OutsideFilter DeleteTags_3 In)
  (connect Branch_1 Output%20A DeleteTags_3 In)
  (connect CreatePolygon2D_1 Out DeleteTags_3 In)
  (connect CreateSpline_1 Out DeleteTags_3 In)
  (connect CreateSpline_3 Out DeleteTags_3 In)
  (connect FilterDataByAttribute_3 InsideFilter DeleteTags_3 In)
  (connect SortTags_1 Out DeleteTags_3 In)
  (connect AttributeFilter_4 InsideFilter FilterDataByAttribute_0 In)
  (connect FilterDataByTag_0 InsideFilter FilterDataByAttribute_1 In)
  (connect FilterDataByType_6 InsideFilter FilterDataByAttribute_2 In)
  (connect AttributeFilter_4 OutsideFilter FilterDataByAttribute_3 In)
  (connect Branch_1 Output%20B FilterDataByIndex_0 In)
  (connect AttributeReduce_0 Out FilterDataByIndex_0 SelectedIndices)
  (connect DefaultInputNode In FilterDataByTag_0 In)
  (connect Branch_0 Output%20B FilterDataByType_1 In)
  (connect Branch_0 Output%20B FilterDataByType_2 In)
  (connect Branch_0 Output%20B FilterDataByType_3 In)
  (connect Branch_0 Output%20A FilterDataByType_4 In)
  (connect Branch_0 Output%20A FilterDataByType_5 In)
  (connect Branch_0 Output%20A FilterDataByType_6 In)
  (connect FilterDataByType_1 InsideFilter FilterElementsByIndex_0 In)
  (connect GetSplineControlPoints_0 Out FilterElementsByIndex_0 In)
  (connect ToPoint_0 Out FilterElementsByIndex_0 In)
  (connect FilterDataByType_2 InsideFilter GetSplineControlPoints_0 In)
  (connect FilterDataByAttribute_2 InsideFilter GetSplineControlPoints_2 In)
  (connect FilterDataByAttribute_2 OutsideFilter GetSplineControlPoints_3 In)
  (connect DeleteTags_0 Out GetTags_0 In)
  (connect GetTags_0 Out MergeAttributes_0 In)
  (connect ExtractAttribute_11 Out Select%20%28Multi%29_0 Boolean)
  (connect ExtractAttribute_1 Out Select%20%28Multi%29_0 Default)
  (connect ExtractAttribute_2 Out Select%20%28Multi%29_0 Double)
  (connect ExtractAttribute_1 Out Select%20%28Multi%29_0 Float)
  (connect ExtractAttribute_3 Out Select%20%28Multi%29_0 Integer%2032)
  (connect ExtractAttribute_4 Out Select%20%28Multi%29_0 Integer%2064)
  (connect ExtractAttribute_13 Out Select%20%28Multi%29_0 Name)
  (connect ExtractAttribute_8 Out Select%20%28Multi%29_0 Quaternion)
  (connect ExtractAttribute_12 Out Select%20%28Multi%29_0 Rotator)
  (connect ExtractAttribute_15 Out Select%20%28Multi%29_0 Soft%20Class%20Path)
  (connect ExtractAttribute_14 Out Select%20%28Multi%29_0 Soft%20Object%20Path)
  (connect ExtractAttribute_10 Out Select%20%28Multi%29_0 String)
  (connect ExtractAttribute_9 Out Select%20%28Multi%29_0 Transform)
  (connect ExtractAttribute_0 Out Select%20%28Multi%29_0 Value)
  (connect ExtractAttribute_6 Out Select%20%28Multi%29_0 Vector)
  (connect ExtractAttribute_5 Out Select%20%28Multi%29_0 Vector%202)
  (connect ExtractAttribute_7 Out Select%20%28Multi%29_0 Vector%204)
  (connect FilterDataByIndex_0 InsideFilter SortTags_0 In)
  (connect FilterDataByIndex_0 OutsideFilter SortTags_1 In)
  (connect FilterDataByType_3 InsideFilter ToPoint_0 In)
  (connect FilterDataByType_4 InsideFilter ToPoint_1 In)
  (connect FilterDataByAttribute_1 InsideFilter WriteDataIndex_0 In)
  (output Out AddTags_0)
  (output Out AddTags_1))
