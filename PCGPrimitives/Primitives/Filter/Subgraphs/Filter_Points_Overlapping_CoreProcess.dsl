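; PCG subgraph that serves as the core process for filtering overlapping points out of point data through Self Pruning.
; It keeps only point data from the input, then a Select on SetExtents passes it on either unchanged or with bounds adjusted by an Extents Modifier using the Extents parameter.
; Pruning type, comparison, randomization and collision settings are extracted from a PruningParameters attribute input and override the Self Pruning node.
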(graph Filter_Points_Overlapping_CoreProcess
  (params (SetExtents :type Bool) (Extents :type Vector :default (1.0 1.0 1.0)))
  ;[0 0 0 0.5] Core Process
  (bind FilterDataByType_2 (Filter%20Data%20By%20Type :TargetType (struct :Ids [(struct :Struct /Script/PCG.PCGDataTypeInfoPoint)])))
  (bind ExtentsModifier_1 (Extents%20Modifier :Extents-param Extents))
  (bind ExtractAttribute_0 (Extract%20Attribute%20at%20Index :bForceOutputAttributeToBeInElementsDomain true :InputSource PruningType :In PruningParameters))
  (bind ExtractAttribute_1 (Extract%20Attribute%20at%20Index :bForceOutputAttributeToBeInElementsDomain true :InputSource ComparisonSource :In PruningParameters))
  (bind ExtractAttribute_2 (Extract%20Attribute%20at%20Index :bForceOutputAttributeToBeInElementsDomain true :InputSource RadiusSimilarityFactor :In PruningParameters))
  (bind ExtractAttribute_3 (Extract%20Attribute%20at%20Index :bForceOutputAttributeToBeInElementsDomain true :InputSource bRandomizedPruning :In PruningParameters))
  (bind ExtractAttribute_4 (Extract%20Attribute%20at%20Index :bForceOutputAttributeToBeInElementsDomain true :InputSource bUseCollisionAttribute :In PruningParameters))
  (bind ExtractAttribute_5 (Extract%20Attribute%20at%20Index :bForceOutputAttributeToBeInElementsDomain true :InputSource CollisionAttribute :In PruningParameters))
  (bind ExtractAttribute_6 (Extract%20Attribute%20at%20Index :bForceOutputAttributeToBeInElementsDomain true :InputSource CollisionQueryFlag :In PruningParameters))
  (bind Select_0 (Select :bUseInputB-param SetExtents))
  (bind SelfPruning_0 (Self%20Pruning))
  (connect FilterDataByType_2 InsideFilter ExtentsModifier_1 In)
  (connect DefaultInputNode In FilterDataByType_2 In)
  (connect FilterDataByType_2 InsideFilter Select_0 Input%20A)
  (connect ExtentsModifier_1 Out Select_0 Input%20B)
  (connect ExtractAttribute_3 Out SelfPruning_0 bRandomizedPruning)
  (connect ExtractAttribute_4 Out SelfPruning_0 bUseCollisionAttribute)
  (connect ExtractAttribute_5 Out SelfPruning_0 CollisionAttribute)
  (connect ExtractAttribute_6 Out SelfPruning_0 CollisionQueryFlag)
  (connect ExtractAttribute_1 Out SelfPruning_0 ComparisonSource)
  (connect Select_0 Out SelfPruning_0 In)
  (connect ExtractAttribute_0 Out SelfPruning_0 PruningType)
  (connect ExtractAttribute_2 Out SelfPruning_0 RadiusSimilarityFactor)
  (output Out SelfPruning_0))
