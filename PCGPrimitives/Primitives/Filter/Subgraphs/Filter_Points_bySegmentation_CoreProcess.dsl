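; PCG subgraph that splits input points into segments by normal, height and color, then filters segments by point count.
; Each optional stage feeds the next, with height measured along each normal group's axis, or world Z without normal grouping.
; Clustering copies the compared value into Position, runs Collapse Points and Attract for group indices, then partitions, isolating unmatched points.
; Backed-up attributes restore overwritten values, and element counts select the output segments.
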(graph Filter_Points_bySegmentation_CoreProcess
  (params (UseNormalSegmentation :type Bool :default true) (NormalSegmentationTolerance :type Double :default 0.25) (UseLocalZPositionSegmentation :type Bool :default true) (LocalZPositionSegmentationTolerance :type Double :default 1e+01) (UseColorSegementation :type Bool :default true) (ColorSegmentationTolerance :type Float :default 1.0) (ElementCountFilterOperator :type (Enum /Script/PCG.EPCGAttributeFilterOperator)) (ElementCountFilterThreshold :type Int))
  (bind FilterDataByType_0 (Filter%20Data%20By%20Type :TargetType (struct :Ids [(struct :Struct /Script/PCG.PCGDataTypeInfoPoint)])))
  ;[0 0 0 0.5 12] Primary/Secondary Inputs
  (bind FilterDataByTag_0 (Filter%20Data%20By%20Tag :SelectedTags "Input:0"))
  ;[0 0 0 0.5] Properties Backup
  (bind CopyAttributes_0 (Copy%20Attributes :InputSource $Position :OutputTarget PositionBackup))
  (bind CopyAttributes_1 (Copy%20Attributes :InputSource $Rotation :OutputTarget RotationBackup))
  (bind CopyAttributes_2 (Copy%20Attributes :InputSource $Color :OutputTarget ColorBackup))
  ;[0 0 0 0.5] Normal Based Segmentation
  (bind Branch_0 (Branch :bOutputToB-param UseNormalSegmentation))
  (bind AttributeMathsOp_0 (Attribute%20Maths%20Op :InputSource1 $Position :InputSource2 $Rotation.up :Operation Set))
  (bind CollapsePoints_0 (Collapse%20Points :AttributesToMerge [] :DistanceThreshold 0.3 :DistanceThreshold-param NormalSegmentationTolerance))
  (bind AttractElement_0 (Attract :bOutputAttractIndex true :Distance 0.3 :OutputAttractIndexAttribute NormalGroup :TargetAttribute $Position :Distance-param NormalSegmentationTolerance))
  (bind AttributeFilter_0 (Filter%20Attribute%20Elements :AttributeTypes (struct :Type Integer32 :Int32Value -1) :bUseConstantThreshold true :Operator Equal :TargetAttribute NormalGroup :ThresholdAttribute @Last))
  (bind AttributePartition_2 (Attribute%20Partition :PartitionAttributeSelectors [$Index]))
  (bind Gather_0 (Gather))
  (bind MakeRotatorAttribute_0 (Make%20Rotator%20Attribute :InputSource1 $Position :Operation MakeRotFromZ :OutputTarget $Rotation))
  (bind AttributeMathsOp_8 (Attribute%20Maths%20Op :InputSource1 $Position :InputSource2 PositionBackup :Operation Set))
  (bind AttributePartition_0 (Attribute%20Partition :PartitionAttributeSelectors [NormalGroup]))
  ;[0 0 0 0.5] LocalZ Based Segmentation
  (bind Branch_2 (Branch :bOutputToB-param UseLocalZPositionSegmentation))
  (bind ExtractAttribute_0 (Extract%20Attribute%20at%20Index :bForceOutputAttributeToBeInElementsDomain true :InputSource $Rotation :OutputAttributeName Rotation))
  (bind MakeTransformAttribute_0 (Make%20Transform%20Attribute :InputSource2 Rotation :OutputTarget ExtractedRotation))
  (bind AttributeVectorOp_1 (Attribute%20Vector%20Op :InputSource1 $Position :Operation InverseTransformLocation))
  (bind AttributeMathsOp_6 (Attribute%20Maths%20Op :InputSource1 $Position.xy :Operation Set))
  (bind CollapsePoints_1 (Collapse%20Points :AttributesToMerge [] :DistanceThreshold 1e+01 :DistanceThreshold-param LocalZPositionSegmentationTolerance))
  (bind AttractElement_1 (Attract :AttractorIndexAttribute $Index :bOutputAttractIndex true :Distance 1e+01 :OutputAttractIndexAttribute LocalZGroup :SourceAndTargetAttributeMapping (map ($Position.z $Position.z)) :TargetAttribute $Position.z :Distance-param LocalZPositionSegmentationTolerance))
  (bind Branch_1 (Branch :bOutputToB-param UseLocalZPositionSegmentation))
  (bind AttributeMathsOp_18 (Attribute%20Maths%20Op :InputSource1 $Position.xy :Operation Set))
  (bind CollapsePoints_7 (Collapse%20Points :AttributesToMerge [] :DistanceThreshold 1e+01 :DistanceThreshold-param LocalZPositionSegmentationTolerance))
  (bind AttractElement_5 (Attract :AttractorIndexAttribute $Index :bOutputAttractIndex true :Distance 1e+01 :OutputAttractIndexAttribute LocalZGroup :SourceAndTargetAttributeMapping (map ($Position.z $Position.z)) :TargetAttribute $Position.z :Distance-param LocalZPositionSegmentationTolerance))
  (bind AttributeMathsOp_19 (Attribute%20Maths%20Op :InputSource1 $Position :InputSource2 PositionBackup :Operation Set))
  (bind AttributeMathsOp_7 (Attribute%20Maths%20Op :InputSource1 $Position :InputSource2 $Position :Operation Set))
  (bind AttributeVectorOp_2 (Attribute%20Vector%20Op :InputSource1 $Position :Operation TransformLocation))
  (bind AttributeFilter_1 (Filter%20Attribute%20Elements :AttributeTypes (struct :Type Integer32 :Int32Value -1) :bUseConstantThreshold true :Operator Equal :TargetAttribute LocalZGroup :ThresholdAttribute @Last))
  (bind AttributePartition_3 (Attribute%20Partition :PartitionAttributeSelectors [$Index]))
  (bind Gather_1 (Gather))
  (bind AttributeMathsOp_9 (Attribute%20Maths%20Op :InputSource1 $Rotation :InputSource2 RotationBackup :Operation Set))
  (bind AttributePartition_1 (Attribute%20Partition :PartitionAttributeSelectors [LocalZGroup]))
  ;[0 0 0 0.5] Color Based Segementation
  (bind Branch_3 (Branch :bOutputToB-param UseColorSegementation))
  (bind AttributeMathsOp_10 (Attribute%20Maths%20Op :InputSource1 $Position :InputSource2 $Color.xyz :Operation Set))
  (bind CollapsePoints_2 (Collapse%20Points :AttributesToMerge [] :DistanceThreshold 1.0 :DistanceThreshold-param ColorSegmentationTolerance))
  (bind AttractElement_2 (Attract :bOutputAttractIndex true :Distance 1.0 :OutputAttractIndexAttribute ColorGroup :TargetAttribute $Position :Distance-param ColorSegmentationTolerance))
  (bind AttributeFilter_2 (Filter%20Attribute%20Elements :AttributeTypes (struct :Type Integer32 :Int32Value -1) :bUseConstantThreshold true :Operator Equal :TargetAttribute ColorGroup :ThresholdAttribute @Last))
  (bind AttributePartition_4 (Attribute%20Partition :PartitionAttributeSelectors [$Index]))
  (bind Gather_2 (Gather))
  (bind AttributeMathsOp_11 (Attribute%20Maths%20Op :InputSource1 $Position :InputSource2 PositionBackup :Operation Set))
  (bind AttributePartition_5 (Attribute%20Partition :PartitionAttributeSelectors [ColorGroup]))
  ;[0 0 0 0.5] Element Count Filtering
  (bind DeleteAttributes_0 (Delete%20Attributes :SelectedAttributes "PositionBackup, RotationBackup, ColorBackup, NormalGroup, LocalZGroup, ColorGroup"))
  (bind GetElementsCount_0 (Get%20Element%20Count))
  (bind CopyAttributes_4 (Copy%20Attributes :InputSource $Index :OutputTarget SourceIndex))
  (bind AttributeFilter_5 (Filter%20Attribute%20Elements :TargetAttribute NumEntries :ThresholdAttribute @Last :Filter ElementCountFilterThreshold :Operator-param ElementCountFilterOperator))
  (bind AttributeReduce_0 (Attribute%20Reduce :InputSource SourceIndex :Operation Join))
  (bind FilterDataByIndex_0 (Filter%20Data%20By%20Index))
  (connect AttributeMathsOp_0 Out AttractElement_0 Source)
  (connect CollapsePoints_0 Out AttractElement_0 Target)
  (connect AttributeMathsOp_6 Out AttractElement_1 Source)
  (connect CollapsePoints_1 Out AttractElement_1 Target)
  (connect AttributeMathsOp_10 Out AttractElement_2 Source)
  (connect CollapsePoints_2 Out AttractElement_2 Target)
  (connect AttributeMathsOp_18 Out AttractElement_5 Source)
  (connect CollapsePoints_7 Out AttractElement_5 Target)
  (connect AttractElement_0 Out AttributeFilter_0 In)
  (connect AttributeMathsOp_19 Out AttributeFilter_1 In)
  (connect AttributeVectorOp_2 Out AttributeFilter_1 In)
  (connect AttractElement_2 Out AttributeFilter_2 In)
  (connect CopyAttributes_4 Out AttributeFilter_5 In)
  (connect Branch_0 Output%20B AttributeMathsOp_0 InA)
  (connect Branch_0 Output%20B AttributeMathsOp_0 InB)
  (connect AttributeVectorOp_1 Out AttributeMathsOp_6 InA)
  (connect AttractElement_1 Out AttributeMathsOp_7 InA)
  (connect AttributeVectorOp_1 Out AttributeMathsOp_7 InB)
  (connect MakeRotatorAttribute_0 Out AttributeMathsOp_8 InA)
  (connect MakeRotatorAttribute_0 Out AttributeMathsOp_8 InB)
  (connect Gather_1 Out AttributeMathsOp_9 InA)
  (connect Gather_1 Out AttributeMathsOp_9 InB)
  (connect Branch_3 Output%20B AttributeMathsOp_10 InA)
  (connect Branch_3 Output%20B AttributeMathsOp_10 InB)
  (connect Gather_2 Out AttributeMathsOp_11 InA)
  (connect Gather_2 Out AttributeMathsOp_11 InB)
  (connect Branch_1 Output%20B AttributeMathsOp_18 InA)
  (connect AttractElement_5 Out AttributeMathsOp_19 InA)
  (connect AttractElement_5 Out AttributeMathsOp_19 InB)
  (connect AttributeMathsOp_8 Out AttributePartition_0 In)
  (connect AttributeMathsOp_9 Out AttributePartition_1 In)
  (connect AttributeFilter_0 InsideFilter AttributePartition_2 In)
  (connect AttributeFilter_1 InsideFilter AttributePartition_3 In)
  (connect AttributeFilter_2 InsideFilter AttributePartition_4 In)
  (connect AttributeMathsOp_11 Out AttributePartition_5 In)
  (connect AttributeFilter_5 InsideFilter AttributeReduce_0 In)
  (connect Branch_2 Output%20B AttributeVectorOp_1 In)
  (connect MakeTransformAttribute_0 Out AttributeVectorOp_1 Transform)
  (connect AttributeMathsOp_7 Out AttributeVectorOp_2 In)
  (connect MakeTransformAttribute_0 Out AttributeVectorOp_2 Transform)
  (connect CopyAttributes_2 Out Branch_0 In)
  (connect Branch_0 Output%20A Branch_1 In)
  (connect AttributePartition_0 Out Branch_2 In)
  (connect AttributePartition_1 Out Branch_3 In)
  (connect Branch_1 Output%20A Branch_3 In)
  (connect Branch_2 Output%20A Branch_3 In)
  (connect AttributeMathsOp_0 Out CollapsePoints_0 In)
  (connect AttributeMathsOp_6 Out CollapsePoints_1 In)
  (connect AttributeMathsOp_10 Out CollapsePoints_2 In)
  (connect AttributeMathsOp_18 Out CollapsePoints_7 In)
  (connect FilterDataByType_0 InsideFilter CopyAttributes_0 Source)
  (connect FilterDataByType_0 InsideFilter CopyAttributes_0 Target)
  (connect CopyAttributes_0 Out CopyAttributes_1 Source)
  (connect CopyAttributes_0 Out CopyAttributes_1 Target)
  (connect CopyAttributes_1 Out CopyAttributes_2 Source)
  (connect CopyAttributes_1 Out CopyAttributes_2 Target)
  (connect GetElementsCount_0 Out CopyAttributes_4 Source)
  (connect GetElementsCount_0 Out CopyAttributes_4 Target)
  (connect AttributePartition_5 Out DeleteAttributes_0 In)
  (connect Branch_3 Output%20A DeleteAttributes_0 In)
  (connect Branch_2 Output%20B ExtractAttribute_0 In)
  (connect DeleteAttributes_0 Out FilterDataByIndex_0 In)
  (connect AttributeReduce_0 Out FilterDataByIndex_0 SelectedIndices)
  (connect DefaultInputNode In FilterDataByTag_0 In)
  (connect FilterDataByTag_0 InsideFilter FilterDataByType_0 In)
  (connect AttributePartition_2 Out Gather_0 In)
  (connect AttributeFilter_0 OutsideFilter Gather_0 In2)
  (connect AttributePartition_3 Out Gather_1 In)
  (connect AttributeFilter_1 OutsideFilter Gather_1 In2)
  (connect AttributePartition_4 Out Gather_2 In)
  (connect AttributeFilter_2 OutsideFilter Gather_2 In2)
  (connect DeleteAttributes_0 Out GetElementsCount_0 In)
  (connect Gather_0 Out MakeRotatorAttribute_0 Z)
  (connect ExtractAttribute_0 Out MakeTransformAttribute_0 Rotation)
  (output Out FilterDataByIndex_0.InsideFilter))
