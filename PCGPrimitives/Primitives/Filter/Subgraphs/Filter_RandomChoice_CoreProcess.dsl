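; PCG core subgraph for random-choice filtering, tagging inputs Selected or Unselected, per dataset or per element, by Ratio or Count.
; Per-dataset ratio uses a loop subgraph and @data.random threshold; count copies seeded noise to datasets, sorts by tag and slices a "0:Count" index range.
; Per-element ratio thresholds noise and merges points; count uses Random Choice.
; All branches share one output; DiscardUnselected is declared but unused here.
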(graph Filter_RandomChoice_CoreProcess
  (params (Ratio :type Double :default 0.5) (Seed :type Int) (RandomPerDataset :type Bool :default true) (DiscardUnselected :type Bool) (Use_Count_instead_of_ratio :type Bool) (Count :type Int :default 4))
  (bind Branch_0 (Branch :bOutputToB-param Use_Count_instead_of_ratio))
  (bind Branch_1 (Branch :bOutputToB-param RandomPerDataset))
  (bind Branch_2 (Branch :bOutputToB-param RandomPerDataset))
  ;[0 0 0 0.5] Per Data Set Methoed
  ;;[0 0 0 0.5] Ratio-based random selection of all input datasets
  (bind Loop_0 (Loop :Graph /PCGPrimitives/Primitives/Filter/Subgraphs/Filter_RandomChoice_CoreProcessLoop.Filter_RandomChoice_CoreProcessLoop :AdditionalSeed Seed))
  (bind FilterDataByAttribute_0 (Filter%20Data%20By%20Attribute :Attribute @data.random :FilterMode FilterByValue :TargetAttribute @data.random :Threshold (struct :bUseConstantThreshold false :AttributeTypes (struct :DoubleValue 0.5)) :Filter Ratio))
  (bind DeleteTags_0 (Delete%20Tags :SelectedTags "Selected, Unselected"))
  (bind AddTags_0 (Add%20Tags :TagsToAdd "Selected"))
  (bind DeleteTags_1 (Delete%20Tags :SelectedTags "Selected, Unselected"))
  (bind AddTags_1 (Add%20Tags :TagsToAdd "Unselected"))
  ;;[0 0 0 0.5] Count-based random selection of all input datasets
  (bind ExtractAttribute_0 (Extract%20Attribute%20at%20Index :bForceOutputAttributeToBeInElementsDomain true))
  (bind MergeAttributes_0 (Merge%20Attributes))
  (bind AddAttribute_0 (Add%20Attribute :OutputTarget random))
  (bind AttributeNoise_1 (Attribute%20Noise :InputSource @Last :Seed-param Seed))
  (bind AttributePartition_0 (Attribute%20Partition :PartitionAttributeSelectors [$Index]))
  (bind CopyAttributes_0 (Copy%20Attributes :InputSource random :OutputTarget @data.Random))
  (bind DataAttributesToTags_0 (Data%20Attributes%20To%20Tags))
  (bind SortTags_0 (Sort%20Data%20By%20Tag%20Value :Tag random))
  (bind AttributeCast_0 (Attribute%20Cast :OutputType String :In Count))
  (bind CreateConstant_0 (Create%20Constant :AttributeTypes (struct :Type String :StringValue "0:")))
  (bind AttributeStringOp_0 (Attribute%20String%20Op))
  (bind FilterDataByIndex_0 (Filter%20Data%20By%20Index))
  (bind DeleteTags_3 (Delete%20Tags :SelectedTags "Selected, Unselected"))
  (bind AddTags_4 (Add%20Tags :TagsToAdd "Selected"))
  (bind DeleteTags_4 (Delete%20Tags :SelectedTags "Selected, Unselected"))
  (bind AddTags_5 (Add%20Tags :TagsToAdd "Unselected"))
  ;[0 0 0 0.5] Per Point Method
  ;;[0 0 0 0.5] Ration-based random selection per point of all input point data
  (bind AddAttribute_1 (Add%20Attribute :OutputTarget Random))
  (bind AttributeNoise_0 (Attribute%20Noise :InputSource Random :Seed-param Seed))
  (bind FilterDataByType_4 (Filter%20Data%20By%20Type :TargetType (struct :Ids [(struct :Struct /Script/PCG.PCGDataTypeInfoPoint) (struct :Struct /Script/PCG.PCGDataTypeInfoParam)])))
  (bind AttributeFilter_1 (Filter%20Attribute%20Elements :bUseSpatialQuery true :TargetAttribute Random :ThresholdAttribute @Last :Filter Ratio))
  (bind FilterDataByType_6 (Filter%20Data%20By%20Type :TargetType (struct :Ids [(struct :Struct /Script/PCG.PCGDataTypeInfoPoint)])))
  (bind MergePoints_0 (Merge%20Points))
  (bind DeleteTags_5 (Delete%20Tags :SelectedTags "Selected, Unselected"))
  (bind AddTags_2 (Add%20Tags :TagsToAdd "Selected"))
  (bind FilterDataByType_5 (Filter%20Data%20By%20Type :TargetType (struct :Ids [(struct :Struct /Script/PCG.PCGDataTypeInfoPoint)])))
  (bind MergePoints_1 (Merge%20Points))
  (bind DeleteTags_6 (Delete%20Tags :SelectedTags "Selected, Unselected"))
  (bind AddTags_3 (Add%20Tags :TagsToAdd "Unselected"))
  ;;[0 0 0 0.5] Count-based random selection per point of all input point data
  (bind FilterDataByType_0 (Filter%20Data%20By%20Type :TargetType (struct :Ids [(struct :Struct /Script/PCG.PCGDataTypeInfoPoint) (struct :Struct /Script/PCG.PCGDataTypeInfoParam)])))
  (bind RandomChoice_0 (Random%20Choice :FixedNumber-param Count :Seed-param Seed))
  (bind DeleteTags_7 (Delete%20Tags :SelectedTags "Selected, Unselected"))
  (bind AddTags_6 (Add%20Tags :TagsToAdd "Selected"))
  (bind DeleteTags_8 (Delete%20Tags :SelectedTags "Selected, Unselected"))
  (bind AddTags_7 (Add%20Tags :TagsToAdd "Unselected"))
  (connect MergeAttributes_0 Out AddAttribute_0 In)
  (connect Branch_1 Output%20A AddAttribute_1 In)
  (connect DeleteTags_0 Out AddTags_0 In)
  (connect DeleteTags_1 Out AddTags_1 In)
  (connect DeleteTags_5 Out AddTags_2 In)
  (connect DeleteTags_6 Out AddTags_3 In)
  (connect DeleteTags_3 Out AddTags_4 In)
  (connect DeleteTags_4 Out AddTags_5 In)
  (connect DeleteTags_7 Out AddTags_6 In)
  (connect DeleteTags_8 Out AddTags_7 In)
  (connect FilterDataByType_4 InsideFilter AttributeFilter_1 In)
  (connect AddAttribute_1 Out AttributeNoise_0 In)
  (connect AddAttribute_0 Out AttributeNoise_1 In)
  (connect AttributeNoise_1 Out AttributePartition_0 In)
  (connect CreateConstant_0 Out AttributeStringOp_0 InA)
  (connect AttributeCast_0 Out AttributeStringOp_0 InB)
  (connect DefaultInputNode In Branch_0 In)
  (connect Branch_0 Output%20A Branch_1 In)
  (connect Branch_0 Output%20B Branch_2 In)
  (connect AttributePartition_0 Out CopyAttributes_0 Source)
  (connect Branch_2 Output%20B CopyAttributes_0 Target)
  (connect CopyAttributes_0 Out DataAttributesToTags_0 In)
  (connect FilterDataByAttribute_0 OutsideFilter DeleteTags_0 In)
  (connect FilterDataByAttribute_0 InsideFilter DeleteTags_1 In)
  (connect FilterDataByIndex_0 InsideFilter DeleteTags_3 In)
  (connect FilterDataByIndex_0 OutsideFilter DeleteTags_4 In)
  (connect MergePoints_0 Out DeleteTags_5 In)
  (connect MergePoints_1 Out DeleteTags_6 In)
  (connect RandomChoice_0 Chosen DeleteTags_7 In)
  (connect RandomChoice_0 Discarded DeleteTags_8 In)
  (connect Branch_2 Output%20B ExtractAttribute_0 In)
  (connect Loop_0 Out FilterDataByAttribute_0 In)
  (connect SortTags_0 Out FilterDataByIndex_0 In)
  (connect AttributeStringOp_0 Out FilterDataByIndex_0 SelectedIndices)
  (connect Branch_2 Output%20A FilterDataByType_0 In)
  (connect AttributeNoise_0 Out FilterDataByType_4 In)
  (connect AttributeFilter_1 InsideFilter FilterDataByType_5 In)
  (connect AttributeFilter_1 OutsideFilter FilterDataByType_6 In)
  (connect Branch_1 Output%20B Loop_0 In)
  (connect ExtractAttribute_0 Out MergeAttributes_0 In)
  (connect FilterDataByType_6 InsideFilter MergePoints_0 In)
  (connect FilterDataByType_5 InsideFilter MergePoints_1 In)
  (connect FilterDataByType_0 InsideFilter RandomChoice_0 In)
  (connect DataAttributesToTags_0 Out SortTags_0 In)
  (output Out AddTags_0)
  (output Out AddTags_1)
  (output Out AddTags_2)
  (output Out AddTags_3)
  (output Out AddTags_4)
  (output Out AddTags_5)
  (output Out AddTags_6)
  (output Out AddTags_7))
