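; PCG loop-body subgraph for a random-choice filter that tags each iteration's data with a random value, without filtering anything itself.
; It adds a `random` attribute in the data domain, then Attribute Noise randomizes it, giving one value per data rather than per element.
; The noise seed combines the loop index with the AdditionalSeed parameter, so it varies per iteration and can be offset by the caller.
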(graph Filter_RandomChoice_CoreProcessLoop
  (params (AdditionalSeed :type Int))
  (bind AddAttribute_1 (Add%20Attribute :OutputTarget @data.random))
  (bind GetLoopIndex_0 (Get%20Loop%20Index))
  (bind AttributeMathsOp_0 (Attribute%20Maths%20Op :InB AdditionalSeed))
  (bind AttributeNoise_1 (Attribute%20Noise :InputSource @data.Random))
  (connect DefaultInputNode In AddAttribute_1 In)
  (connect GetLoopIndex_0 Out AttributeMathsOp_0 InA)
  (connect AddAttribute_1 Out AttributeNoise_1 In)
  (connect AttributeMathsOp_0 Out AttributeNoise_1 Seed)
  (output Out AttributeNoise_1))
