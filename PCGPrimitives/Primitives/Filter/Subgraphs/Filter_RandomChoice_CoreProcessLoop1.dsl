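; PCG loop-body subgraph for the Filter_RandomChoice primitive that gives each iteration's input data a seeded random value.
; It adds a 'random' attribute and runs Attribute Noise on it, seeded by combining the loop index with the additional_seed parameter.
; The result is then copied onto the same data as the data-domain attribute @data.random and output; no filtering happens in this graph.
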(graph Filter_RandomChoice_CoreProcessLoop1
  (params (additional_seed :type Double))
  (bind AddAttribute_1 (Add%20Attribute :OutputTarget random))
  (bind GetLoopIndex_0 (Get%20Loop%20Index))
  (bind AttributeMathsOp_0 (Attribute%20Maths%20Op :InB additional_seed))
  (bind AttributeNoise_1 (Attribute%20Noise :InputSource Random))
  (bind CopyAttributes_0 (Copy%20Attributes :InputSource random :OutputTarget @data.random))
  (connect DefaultInputNode In AddAttribute_1 In)
  (connect GetLoopIndex_0 Out AttributeMathsOp_0 InA)
  (connect AddAttribute_1 Out AttributeNoise_1 In)
  (connect AttributeMathsOp_0 Out AttributeNoise_1 Seed)
  (connect AttributeNoise_1 Out CopyAttributes_0 Source)
  (connect AttributeNoise_1 Out CopyAttributes_0 Target)
  (output Out CopyAttributes_0))
