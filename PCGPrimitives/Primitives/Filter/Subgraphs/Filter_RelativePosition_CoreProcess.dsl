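; PCG subgraph that filters points, splines and polygons by position relative to reference points: directional, nearest or farthest.
; Splines and polygons become segment points; elements are scored individually or combined per data, by direction dot product (optionally in reference rotation) or distance.
; Sorting keeps the top element, or a tolerance threshold keeps a range; Bidirectional adds the opposite end.
; Whole-data results map back by index, and segments are rebuilt as linear splines.
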(graph Filter_RelativePosition_CoreProcess
  (params (RelativePosition :type (Enum /PCGPrimitives/Primitives/Filter/Resources/EPCG_RelativePositionEnum.EPCG_RelativePositionEnum)) (Bidirectional :type Bool) (UseDirectionVector :type Bool) (Direction :type Vector :default (1.0 0.0 0.0)) (UseLocalPointRotation :type Bool :default true) (ToleranceThreshold :type Double) (UseToleranceThreshold :type Bool) (FilterPerElement :type Bool :default true))
  (bind FilterDataByTag_1 (Filter%20Data%20By%20Tag :SelectedTags "Input:0"))
  (bind FilterDataByType_3 (Filter%20Data%20By%20Type :TargetType (struct :Ids [(struct :Struct /Script/PCG.PCGDataTypeInfoPoint)])))
  (bind ToPoint_1 (To%20Point))
  (bind AttributeMathsOp_1 (Attribute%20Maths%20Op :InputSource1 $Position :InputSource2 $Position :Operation Subtract :OutputTarget RelativePosition)) ; Write the Relative position vector.
  (bind AttributeCompareOp_0 (Attribute%20Compare%20Op))
  (bind AttributeCompareOp_1 (Attribute%20Compare%20Op))
  (bind AttributeBooleanOp_0 (Attribute%20Boolean%20Op :Operation Or))
  (bind Branch_1 (Branch))
  (bind AttributeRename_0 (Attribute%20Rename :AttributeToRename Direction :NewAttributeName RelativePosition))
  (bind Branch_7 (Branch))
  (bind MakeRotatorAttribute_0 (Make%20Rotator%20Attribute :InputSource1 RelativePosition :OutputTarget RelativePosition))
  (bind Branch_13 (Branch :bOutputToB-param UseLocalPointRotation))
  (bind ExtractAttribute_1 (Extract%20Attribute%20at%20Index :bForceOutputAttributeToBeInElementsDomain true :InputSource $Rotation))
  (bind AttributeRotatorOp_1 (Attribute%20Rotator%20Op :InputSource1 RelativePosition :InputSource2 @Last))
  (bind ExtractAttribute_2 (Extract%20Attribute%20at%20Index :bForceOutputAttributeToBeInElementsDomain true :InputSource RelativePosition.forward :OutputAttributeName RelativePosition))
  (bind Branch_12 (Branch :bOutputToB-param FilterPerElement))
  (bind MergePoints_4 (Merge%20Points))
  (bind SortAttributes_0 (Sort%20Attributes :bUseStableSort false :SortAttributes [(struct :InputSource RelativePosition :SortMethod Descending)]))
  (bind Branch_0 (Branch :bOutputToB-param Bidirectional))
  (bind Branch_11 (Branch :bOutputToB-param FilterPerElement))
  (bind MergePoints_2 (Merge%20Points))
  (bind Branch_4 (Branch))
  (bind SortAttributes_1 (Sort%20Attributes :bUseStableSort false :SortAttributes [(struct :InputSource RelativePosition)]))
  (bind SortAttributes_2 (Sort%20Attributes :bUseStableSort false :SortAttributes [(struct :InputSource RelativePosition :SortMethod Descending)]))
  (bind Branch_2 (Branch :bOutputToB-param UseToleranceThreshold))
  (bind Branch_3 (Branch :bOutputToB-param UseToleranceThreshold))
  (bind FilterElementsByIndex_0 (Filter%20Elements%20By%20Index :bSelectIndicesByInput false :SelectedIndices "0"))
  (bind Branch_9 (Branch :bOutputToB-param FilterPerElement))
  (bind PointToAttributeSet_0 (Point%20To%20Attribute%20Set))
  (bind DeleteAttributes_1 (Delete%20Attributes :Operation KeepSelectedAttributes :SelectedAttributes "FilterRelativeDataIndex"))
  (bind AttributeReduce_2 (Attribute%20Reduce :Operation Join))
  (bind Branch_14 (Branch :bOutputToB-param UseToleranceThreshold))
  (bind FilterDataByIndex_0 (Filter%20Data%20By%20Index))
  (bind GetElementsCount_0 (Get%20Element%20Count))
  (bind AttributeMathsOp_2 (Attribute%20Maths%20Op :Operation Subtract))
  (bind FilterElementsByIndex_1 (Filter%20Elements%20By%20Index :bSelectIndicesByInput false :SelectedIndices "0"))
  (bind Branch_10 (Branch :bOutputToB-param FilterPerElement))
  (bind PointToAttributeSet_1 (Point%20To%20Attribute%20Set))
  (bind DeleteAttributes_2 (Delete%20Attributes :Operation KeepSelectedAttributes :SelectedAttributes "FilterRelativeDataIndex"))
  (bind AttributeReduce_3 (Attribute%20Reduce :Operation Join))
  (bind FilterDataByIndex_1 (Filter%20Data%20By%20Index))
  ;[0 0 0 0.5] Convert all input data into points for filtering them per element or combined per data. Splines and polys are reconstructed as spline segments after filtering. Points are passed as points.
  (bind FilterDataByType_0 (Filter%20Data%20By%20Type :TargetType (struct :Ids [(struct :Struct /Script/PCG.PCGDataTypeInfoPoint)])))
  (bind FilterDataByType_1 (Filter%20Data%20By%20Type :TargetType (struct :Ids [(struct :Struct /Script/PCG.PCGDataTypeInfoSpline)])))
  (bind FilterDataByType_2 (Filter%20Data%20By%20Type :TargetType (struct :Ids [(struct :Struct /Script/PCG.PCGDataTypeInfoPolygon2D)])))
  (bind ToPoint_0 (To%20Point))
  (bind CreateSpline_1 (Create%20Spline :bClosedLoop true :bLinear true :InterpTypeAttribute None))
  (bind SplineToSegment_0 (Spline%20to%20Segment))
  (bind Branch_8 (Branch :bOutputToB-param FilterPerElement))
  (bind CombinePoints_0 (Combine%20Points :bUseFirstPointTransform false))
  (bind WriteDataIndex_0 (Write%20Data%20Index :bAddAttribute true :bAddTag false :IndexAttribute FilterRelativeDataIndex))
  ;[0 0 0 0.5] Relative position mode selection.
  (bind AttributeVectorOp_2 (Attribute%20Vector%20Op :Operation Normalize :In Direction))
  (bind CreateConstant_0 (Create%20Constant :AttributeTypes (struct :Type Vector :VectorValue (1.0 0.0 0.0)) :OutputTarget RelativePosition))
  (bind CreateConstant_1 (Create%20Constant :AttributeTypes (struct :Type Vector :VectorValue (-1.0 0.0 0.0)) :OutputTarget RelativePosition))
  (bind CreateConstant_2 (Create%20Constant :AttributeTypes (struct :Type Vector :VectorValue (0.0 1.0 0.0)) :OutputTarget RelativePosition))
  (bind CreateConstant_3 (Create%20Constant :AttributeTypes (struct :Type Vector :VectorValue (0.0 -1.0 0.0)) :OutputTarget RelativePosition))
  (bind CreateConstant_4 (Create%20Constant :AttributeTypes (struct :Type Vector :VectorValue (0.0 0.0 1.0)) :OutputTarget RelativePosition))
  (bind CreateConstant_5 (Create%20Constant :AttributeTypes (struct :Type Vector :VectorValue (0.0 0.0 -1.0)) :OutputTarget RelativePosition))
  (bind CreateConstant_8 (Create%20Constant :AttributeTypes (struct :Type String :StringValue "Nearest") :OutputTarget RelativePosition))
  (bind CreateConstant_9 (Create%20Constant :AttributeTypes (struct :Type String :StringValue "Farthest") :OutputTarget RelativePosition))
  (bind Select%20%28Multi%29_2 (Select%20%28Multi%29 :EnumSelection (struct :Class /PCGPrimitives/Primitives/Filter/Resources/EPCG_RelativePositionEnum.EPCG_RelativePositionEnum) :SelectionMode Enum :Value RelativePosition))
  (bind Select_0 (Select :bUseInputB-param UseDirectionVector))
  ;[0 0 0 0.5] Compute Position or Distance based attribute
  (bind AttributeVectorOp_0 (Attribute%20Vector%20Op :InputSource1 RelativePosition.normalized :InputSource2 @Last.normalized :Operation Dot :OutputTarget RelativePosition))
  (bind AttributeVectorOp_1 (Attribute%20Vector%20Op :InputSource1 RelativePosition :Operation Length))
  ;[0 0 0 0.5] Tolerance Threshold Filtering path
  (bind AttributeMathsOp_0 (Attribute%20Maths%20Op :Operation OneMinus :In ToleranceThreshold))
  (bind Branch_5 (Branch))
  (bind Branch_6 (Branch))
  (bind AttributeFilter_0 (Filter%20Attribute%20Elements :bGenerateOutputDataEvenIfEmpty false :Operator GreaterOrEqual :TargetAttribute RelativePosition :ThresholdAttribute @Last))
  (bind AttributeFilter_1 (Filter%20Attribute%20Elements :bGenerateOutputDataEvenIfEmpty false :Operator LesserOrEqual :TargetAttribute RelativePosition :ThresholdAttribute @Last :Filter ToleranceThreshold))
  (bind MergePoints_3 (Merge%20Points))
  (bind AttributeReduce_0 (Attribute%20Reduce :InputSource RelativePosition :Operation Max :OutputAttributeName RelativePosition))
  (bind AttributeReduce_1 (Attribute%20Reduce :InputSource RelativePosition :Operation Min :OutputAttributeName RelativePosition))
  (bind AttributeMathsOp_6 (Attribute%20Maths%20Op :Operation Subtract))
  (bind AttributeMathsOp_4 (Attribute%20Maths%20Op :Operation Multiply :InA ToleranceThreshold))
  (bind AttributeMathsOp_7 (Attribute%20Maths%20Op))
  (bind AttributeMathsOp_3 (Attribute%20Maths%20Op :Operation Multiply))
  (bind AttributeMathsOp_8 (Attribute%20Maths%20Op))
  (bind Select_2 (Select))
  (bind CreateConstant_10 (Create%20Constant :AttributeTypes (struct :Type Integer32 :Int32Value 1)))
  (bind CreateConstant_11 (Create%20Constant :AttributeTypes (struct :Type Integer32 :Int32Value 3)))
  (bind Select_4 (Select))
  (bind AttributeFilter_2 (Filter%20Attribute%20Elements :bGenerateOutputDataEvenIfEmpty false :Operator GreaterOrEqual :TargetAttribute RelativePosition :ThresholdAttribute @Last))
  (bind Select_3 (Select))
  (bind Select_5 (Select))
  (bind AttributeFilter_3 (Filter%20Attribute%20Elements :bGenerateOutputDataEvenIfEmpty false :Operator GreaterOrEqual :TargetAttribute RelativePosition :ThresholdAttribute @Last))
  ;[0 0 0 0.5] Reconstruct filtered spline segments.
  (bind FilterDataByTag_0 (Filter%20Data%20By%20Tag :SelectedTags "Type:Point"))
  (bind ResetPointCenter_0 (Reset%20Point%20Center :PointCenterLocation (0.0 0.0 0.0)))
  (bind ResetPointCenter_1 (Reset%20Point%20Center :PointCenterLocation (1.0 0.0 0.0)))
  (bind MergePoints_0 (Merge%20Points))
  (bind FilterDataByTag_2 (Filter%20Data%20By%20Tag :SelectedTags "Type:Point"))
  (bind ResetPointCenter_2 (Reset%20Point%20Center :PointCenterLocation (0.0 0.0 0.0)))
  (bind ResetPointCenter_3 (Reset%20Point%20Center :PointCenterLocation (1.0 0.0 0.0)))
  (bind MergePoints_1 (Merge%20Points))
  (bind AttributePartition_0 (Attribute%20Partition :PartitionAttributeSelectors [SegmentIndex]))
  (bind CreateSpline_0 (Create%20Spline :bLinear true :InterpTypeAttribute None))
  (bind DeleteAttributes_0 (Delete%20Attributes :SelectedAttributes "RelativePosition"))
  (reroute DistanceBased AttributeBooleanOp_0)
  (reroute Farthest AttributeCompareOp_1)
  (reroute Source%20Point%20Data Branch_8.Output%20A)
  (connect AttributeCompareOp_0 Out AttributeBooleanOp_0 InA)
  (connect AttributeCompareOp_1 Out AttributeBooleanOp_0 InB)
  (connect Select_0 Out AttributeCompareOp_0 InA)
  (connect Select_0 Out AttributeCompareOp_1 InA)
  (connect AttributeMathsOp_0 Out AttributeFilter_0 Filter)
  (connect Branch_5 Output%20A AttributeFilter_0 In)
  (connect Branch_6 Output%20A AttributeFilter_0 In)
  (connect Branch_6 Output%20A AttributeFilter_1 In)
  (connect Select_2 Out AttributeFilter_2 Filter)
  (connect Branch_5 Output%20B AttributeFilter_2 In)
  (connect Branch_6 Output%20B AttributeFilter_2 In)
  (connect Select_4 Out AttributeFilter_2 Operator)
  (connect Select_3 Out AttributeFilter_3 Filter)
  (connect Branch_6 Output%20B AttributeFilter_3 In)
  (connect Select_5 Out AttributeFilter_3 Operator)
  (connect Branch_8 Output%20B AttributeMathsOp_1 InA)
  (connect WriteDataIndex_0 Out AttributeMathsOp_1 InA)
  (connect ToPoint_1 Out AttributeMathsOp_1 InB)
  (connect GetElementsCount_0 Out AttributeMathsOp_2 InA)
  (connect AttributeMathsOp_0 Out AttributeMathsOp_3 InA)
  (connect AttributeMathsOp_6 Out AttributeMathsOp_3 InB)
  (connect AttributeMathsOp_6 Out AttributeMathsOp_4 InB)
  (connect AttributeReduce_0 Out AttributeMathsOp_6 InA)
  (connect AttributeReduce_1 Out AttributeMathsOp_6 InB)
  (connect AttributeMathsOp_4 Out AttributeMathsOp_7 InA)
  (connect AttributeReduce_1 Out AttributeMathsOp_7 InB)
  (connect AttributeMathsOp_3 Out AttributeMathsOp_8 InA)
  (connect AttributeReduce_1 Out AttributeMathsOp_8 InB)
  (connect MergePoints_0 Out AttributePartition_0 In)
  (connect MergePoints_1 Out AttributePartition_0 In)
  (connect MergePoints_3 Out AttributeReduce_0 In)
  (connect MergePoints_3 Out AttributeReduce_1 In)
  (connect DeleteAttributes_1 Out AttributeReduce_2 In)
  (connect DeleteAttributes_2 Out AttributeReduce_3 In)
  (connect Select_0 Out AttributeRename_0 In)
  (connect Branch_13 Output%20B AttributeRotatorOp_1 InA)
  (connect ExtractAttribute_1 Out AttributeRotatorOp_1 InB)
  (connect Branch_1 Output%20A AttributeVectorOp_0 InA)
  (connect ExtractAttribute_2 Out AttributeVectorOp_0 InB)
  (connect Branch_1 Output%20B AttributeVectorOp_1 In)
  (connect SortAttributes_0 Out Branch_0 In)
  (connect AttributeBooleanOp_0 Out Branch_1 bOutputToB)
  (connect AttributeMathsOp_1 Out Branch_1 In)
  (connect Branch_0 Output%20A Branch_2 In)
  (connect SortAttributes_1 Out Branch_2 In)
  (connect SortAttributes_2 Out Branch_2 In)
  (connect Branch_0 Output%20B Branch_3 In)
  (connect AttributeCompareOp_1 Out Branch_4 bOutputToB)
  (connect Branch_11 Output%20B Branch_4 In)
  (connect MergePoints_2 Out Branch_4 In)
  (connect DistanceBased Out Branch_5 bOutputToB)
  (connect Branch_2 Output%20B Branch_5 In)
  (connect DistanceBased Out Branch_6 bOutputToB)
  (connect Branch_3 Output%20B Branch_6 In)
  (connect AttributeBooleanOp_0 Out Branch_7 bOutputToB)
  (connect AttributeRename_0 Out Branch_7 In)
  (connect FilterDataByType_0 InsideFilter Branch_8 In)
  (connect SplineToSegment_0 Out Branch_8 In)
  (connect AttributeFilter_0 InsideFilter Branch_9 In)
  (connect AttributeFilter_1 InsideFilter Branch_9 In)
  (connect AttributeFilter_2 InsideFilter Branch_9 In)
  (connect AttributeFilter_3 InsideFilter Branch_9 In)
  (connect FilterElementsByIndex_0 In%20Filter Branch_9 In)
  (connect AttributeFilter_1 InsideFilter Branch_10 In)
  (connect AttributeFilter_3 InsideFilter Branch_10 In)
  (connect FilterElementsByIndex_1 In%20Filter Branch_10 In)
  (connect AttributeVectorOp_1 Out Branch_11 In)
  (connect AttributeVectorOp_0 Out Branch_12 In)
  (connect MakeRotatorAttribute_0 Out Branch_13 In)
  (connect Source%20Point%20Data Out Branch_14 In)
  (connect Branch_8 Output%20A CombinePoints_0 In)
  (connect AttributePartition_0 Out CreateSpline_0 In)
  (connect ToPoint_0 Out CreateSpline_1 In)
  (connect CreateSpline_0 Out DeleteAttributes_0 In)
  (connect FilterDataByTag_0 InsideFilter DeleteAttributes_0 In)
  (connect FilterDataByTag_2 InsideFilter DeleteAttributes_0 In)
  (connect PointToAttributeSet_0 Out DeleteAttributes_1 In)
  (connect PointToAttributeSet_1 Out DeleteAttributes_2 In)
  (connect ToPoint_1 Out ExtractAttribute_1 In)
  (connect AttributeRotatorOp_1 Out ExtractAttribute_2 In)
  (connect Branch_13 Output%20A ExtractAttribute_2 In)
  (connect Branch_14 Output%20A FilterDataByIndex_0 In)
  (connect AttributeReduce_2 Out FilterDataByIndex_0 SelectedIndices)
  (connect Branch_14 Output%20B FilterDataByIndex_1 In)
  (connect AttributeReduce_3 Out FilterDataByIndex_1 SelectedIndices)
  (connect Branch_9 Output%20B FilterDataByTag_0 In)
  (connect FilterDataByIndex_0 InsideFilter FilterDataByTag_0 In)
  (connect DefaultInputNode In FilterDataByTag_1 In)
  (connect Branch_10 Output%20B FilterDataByTag_2 In)
  (connect FilterDataByIndex_1 InsideFilter FilterDataByTag_2 In)
  (connect FilterDataByTag_1 InsideFilter FilterDataByType_0 In)
  (connect FilterDataByTag_1 InsideFilter FilterDataByType_1 In)
  (connect FilterDataByTag_1 InsideFilter FilterDataByType_2 In)
  (connect FilterDataByTag_1 OutsideFilter FilterDataByType_3 In)
  (connect Branch_2 Output%20A FilterElementsByIndex_0 In)
  (connect Branch_3 Output%20A FilterElementsByIndex_0 In)
  (connect Branch_3 Output%20A FilterElementsByIndex_1 In)
  (connect AttributeMathsOp_2 Out FilterElementsByIndex_1 SelectedIndices)
  (connect Branch_3 Output%20A GetElementsCount_0 In)
  (connect Branch_7 Output%20A MakeRotatorAttribute_0 Forward)
  (connect ResetPointCenter_0 Out MergePoints_0 In)
  (connect ResetPointCenter_1 Out MergePoints_0 In2)
  (connect ResetPointCenter_2 Out MergePoints_1 In)
  (connect ResetPointCenter_3 Out MergePoints_1 In2)
  (connect Branch_11 Output%20A MergePoints_2 In)
  (connect Branch_5 Output%20B MergePoints_3 In)
  (connect Branch_12 Output%20A MergePoints_4 In)
  (connect Branch_9 Output%20A PointToAttributeSet_0 In)
  (connect Branch_10 Output%20A PointToAttributeSet_1 In)
  (connect FilterDataByTag_0 OutsideFilter ResetPointCenter_0 In)
  (connect FilterDataByTag_0 OutsideFilter ResetPointCenter_1 In)
  (connect FilterDataByTag_2 OutsideFilter ResetPointCenter_2 In)
  (connect FilterDataByTag_2 OutsideFilter ResetPointCenter_3 In)
  (connect Select%20%28Multi%29_2 Out Select_0 Input%20A)
  (connect AttributeVectorOp_2 Out Select_0 Input%20B)
  (connect Farthest Out Select_2 bUseInputB)
  (connect AttributeMathsOp_7 Out Select_2 Input%20A)
  (connect AttributeMathsOp_8 Out Select_2 Input%20B)
  (connect Farthest Out Select_3 bUseInputB)
  (connect AttributeMathsOp_8 Out Select_3 Input%20A)
  (connect AttributeMathsOp_7 Out Select_3 Input%20B)
  (connect Farthest Out Select_4 bUseInputB)
  (connect CreateConstant_11 Out Select_4 Input%20A)
  (connect CreateConstant_10 Out Select_4 Input%20B)
  (connect Farthest Out Select_5 bUseInputB)
  (connect CreateConstant_10 Out Select_5 Input%20A)
  (connect CreateConstant_11 Out Select_5 Input%20B)
  (connect CreateConstant_1 Out Select%20%28Multi%29_2 Back)
  (connect CreateConstant_5 Out Select%20%28Multi%29_2 Bottom)
  (connect CreateConstant_0 Out Select%20%28Multi%29_2 Default)
  (connect CreateConstant_9 Out Select%20%28Multi%29_2 Farthest)
  (connect CreateConstant_0 Out Select%20%28Multi%29_2 Front)
  (connect CreateConstant_3 Out Select%20%28Multi%29_2 Left)
  (connect CreateConstant_8 Out Select%20%28Multi%29_2 Nearest)
  (connect CreateConstant_2 Out Select%20%28Multi%29_2 Right)
  (connect CreateConstant_4 Out Select%20%28Multi%29_2 Top)
  (connect Branch_12 Output%20B SortAttributes_0 In)
  (connect MergePoints_4 Out SortAttributes_0 In)
  (connect Branch_4 Output%20A SortAttributes_1 In)
  (connect Branch_4 Output%20B SortAttributes_2 In)
  (connect CreateSpline_1 Out SplineToSegment_0 In)
  (connect FilterDataByType_1 InsideFilter SplineToSegment_0 In)
  (connect FilterDataByType_2 InsideFilter ToPoint_0 In)
  (connect FilterDataByType_3 InsideFilter ToPoint_1 In)
  (connect CombinePoints_0 Out WriteDataIndex_0 In)
  (output Out DeleteAttributes_0))
