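; PCG subgraph that prepares the outputs of the Relative Position filter primitive by tagging them for output.
; It removes empty data, then splits the rest by the Type:Point tag.
; Each side's data count drives a Branch, so only a non-empty set reaches its own instance of the shared OutputTaggingSG subgraph.
; Both tagged results merge into the single Out pin.
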(graph Filter_RelativePosition_PrepareOutputs
  (bind RemoveEmptyData_0 (Remove%20Empty%20Data))
  (bind FilterDataByTag_1 (Filter%20Data%20By%20Tag :SelectedTags "Type:Point"))
  ;[0 0 0 0.5] Tag and apply Metadata for Output
  (bind DataNum_0 (Data%20Count))
  (bind Branch_0 (Branch))
  (bind Subgraph_0 (Subgraph :Graph /PCGPrimitives/Primitives/Shared/PrepareOutputs/OutputTaggingSG.OutputTaggingSG))
  (bind DataNum_1 (Data%20Count))
  (bind Branch_1 (Branch))
  (bind Subgraph_1 (Subgraph :Graph /PCGPrimitives/Primitives/Shared/PrepareOutputs/OutputTaggingSG.OutputTaggingSG))
  (connect DataNum_0 Out Branch_0 bOutputToB)
  (connect FilterDataByTag_1 InsideFilter Branch_0 In)
  (connect DataNum_1 Out Branch_1 bOutputToB)
  (connect FilterDataByTag_1 OutsideFilter Branch_1 In)
  (connect FilterDataByTag_1 InsideFilter DataNum_0 In)
  (connect FilterDataByTag_1 OutsideFilter DataNum_1 In)
  (connect RemoveEmptyData_0 Out FilterDataByTag_1 In)
  (connect DefaultInputNode In RemoveEmptyData_0 In)
  (connect Branch_0 Output%20B Subgraph_0 In)
  (connect Branch_1 Output%20B Subgraph_1 In)
  (output Out Subgraph_0)
  (output Out Subgraph_1))
