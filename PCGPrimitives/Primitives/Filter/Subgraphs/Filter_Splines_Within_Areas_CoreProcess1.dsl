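; PCG subgraph that forms a core step of filtering splines by whether they lie within areas.
; It cleans attributes and selection tags, then splits input by the Input:0 tag into open splines and, from the remainder, 2D polygons.
; An inner loop subgraph receives the splines with the polygons as clipping areas, plus a flag for including partially contained areas, and supplies the output.
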(graph Filter_Splines_Within_Areas_CoreProcess1
  (params (include_partially_contained_areas :type Bool))
  (bind DeleteAttributes_0 (Delete%20Attributes :Operation KeepSelectedAttributes))
  (bind DeleteTags_1 (Delete%20Tags :SelectedTags "selected, unselected"))
  (bind FilterDataByTag_0 (Filter%20Data%20By%20Tag :SelectedTags "Input:0"))
  (bind FilterDataByType_0 (Filter%20Data%20By%20Type :TargetType (struct :Ids [(struct :Struct /Script/PCG.PCGDataTypeInfoSpline)])))
  (bind FilterDataByAttribute_0 (Filter%20Data%20By%20Attribute :FilterMode FilterByValue :FilterOperator Equal :TargetAttribute @Data.$IsClosed :Threshold (struct :AttributeTypes (struct :Type Boolean))))
  (bind FilterDataByType_4 (Filter%20Data%20By%20Type :TargetType (struct :Ids [(struct :Struct /Script/PCG.PCGDataTypeInfoPolygon2D)])))
  (bind Loop_0 (Loop :Graph /PCGPrimitives/Primitives/Filter/Subgraphs/Filter_splines_Within_Areas_CoreProcess_innerLoop1.Filter_splines_Within_Areas_CoreProcess_innerLoop1 :include_partially_contained_areas include_partially_contained_areas))
  (connect DefaultInputNode In DeleteAttributes_0 In)
  (connect DeleteAttributes_0 Out DeleteTags_1 In)
  (connect FilterDataByType_0 InsideFilter FilterDataByAttribute_0 In)
  (connect DeleteTags_1 Out FilterDataByTag_0 In)
  (connect FilterDataByTag_0 InsideFilter FilterDataByType_0 In)
  (connect FilterDataByTag_0 OutsideFilter FilterDataByType_4 In)
  (connect FilterDataByType_4 InsideFilter Loop_0 Clipping%20polygons)
  (connect FilterDataByAttribute_0 InsideFilter Loop_0 In)
  (output Out Loop_0))
