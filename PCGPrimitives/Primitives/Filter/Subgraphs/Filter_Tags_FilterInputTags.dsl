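; PCG filter subgraph that splits input data by tag into InsideFilter and OutsideFilter outputs, marked Selected or Unselected.
; A comparison on FilteredInputTags drives a Branch that either runs Filter Data By Tag or bypasses it straight to the selected path, apparently for an empty tag string.
; InverFilter drives the filter's operation.
; Both paths delete existing Selected and Unselected tags before adding their own, so earlier selection marks do not carry over.
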(graph Filter_Tags_FilterInputTags
  (params (FilteredInputTags :type String) (InverFilter :type Bool))
  (outputs (InsideFilter :type Any) (OutsideFilter :type Any))
  (bind AttributeCompareOp_0 (Attribute%20Compare%20Op :InA FilteredInputTags))
  (bind Branch_1 (Branch))
  (bind FilterDataByTag_1 (Filter%20Data%20By%20Tag :Operation-param InverFilter :SelectedTags-param FilteredInputTags))
  (bind DeleteTags_1 (Delete%20Tags :SelectedTags "Selected, Unselected"))
  (bind AddTags_1 (Add%20Tags :TagsToAdd "Selected"))
  (bind DeleteTags_2 (Delete%20Tags :SelectedTags "Selected, Unselected"))
  (bind AddTags_2 (Add%20Tags :TagsToAdd "Unselected"))
  (connect DeleteTags_1 Out AddTags_1 In)
  (connect DeleteTags_2 Out AddTags_2 In)
  (connect AttributeCompareOp_0 Out Branch_1 bOutputToB)
  (connect DefaultInputNode In Branch_1 In)
  (connect Branch_1 Output%20B DeleteTags_1 In)
  (connect FilterDataByTag_1 InsideFilter DeleteTags_1 In)
  (connect FilterDataByTag_1 OutsideFilter DeleteTags_2 In)
  (connect Branch_1 Output%20A FilterDataByTag_1 In)
  (output InsideFilter AddTags_1)
  (output OutsideFilter AddTags_2))
