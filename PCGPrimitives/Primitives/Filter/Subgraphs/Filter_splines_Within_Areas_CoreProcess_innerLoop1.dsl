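; PCG inner-loop subgraph that tags a spline "selected" or "unselected" by whether it lies within clipping polygons.
; Clip Paths cuts the spline, and an empty result routes the original to "unselected".
; Otherwise, matching floored original and clipped lengths mark it fully contained and selected.
; Differing lengths mean partial containment, which the include_partially_contained_areas parameter resolves to either tag.
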(graph Filter_splines_Within_Areas_CoreProcess_innerLoop1
  (params (include_partially_contained_areas :type Bool))
  (inputs (Clipping%20polygons :type Polygon2D))
  (bind ClipPaths_1 (Clip%20Paths))
  (bind GetElementsCount_0 (Get%20Element%20Count))
  (bind Branch_1 (Branch)) ; IF no data go unselected
  (bind AddTags_1 (Add%20Tags :TagsToAdd "unselected"))
  ;[0.15 0.15 0.15 0.5] COMPARING SPLINE after intersection
  (bind AttributeMathsOp_0 (Attribute%20Maths%20Op :InputSource1 @Data.$Length :Operation Floor :OutputTarget @data.lenghtFloored))
  (bind AttributeMathsOp_1 (Attribute%20Maths%20Op :InputSource1 @Data.$Length :Operation Floor :OutputTarget @data.lenghtFloored))
  (bind AttributeCompareOp_0 (Attribute%20Compare%20Op :InputSource1 @data.lenghtFloored :InputSource2 @data.lenghtFloored :OutputTarget @data.untouched))
  (bind ExtractAttribute_0 (Extract%20Attribute%20at%20Index :bForceOutputAttributeToBeInElementsDomain true :InputSource @data.untouched))
  (bind Branch_0 (Branch))
  (bind Branch_2 (Branch :bOutputToB-param include_partially_contained_areas))
  (bind AddTags_3 (Add%20Tags :TagsToAdd "selected"))
  (bind AddTags_5 (Add%20Tags :TagsToAdd "selected"))
  (bind AddTags_0 (Add%20Tags :TagsToAdd "unselected"))
  (bind ExtractAttribute_1 (Extract%20Attribute%20at%20Index :bForceOutputAttributeToBeInElementsDomain true :InputSource @Data.lenghtFloored))
  (bind MergeAttributes_0 (Merge%20Attributes))
  (reroute Input%20spline In)
  (connect Branch_2 Output%20A AddTags_0 In)
  (connect Branch_1 Output%20A AddTags_1 In)
  (connect Branch_2 Output%20B AddTags_3 In)
  (connect Branch_0 Output%20B AddTags_5 In)
  (connect AttributeMathsOp_1 Out AttributeCompareOp_0 InA)
  (connect AttributeMathsOp_0 Out AttributeCompareOp_0 InB)
  (connect Input%20spline Out AttributeMathsOp_0 In)
  (connect ClipPaths_1 Out AttributeMathsOp_1 In)
  (connect ExtractAttribute_0 Out Branch_0 bOutputToB)
  (connect Branch_1 Output%20B Branch_0 In)
  (connect GetElementsCount_0 Out Branch_1 bOutputToB)
  (connect Input%20spline Out Branch_1 In)
  (connect Branch_0 Output%20A Branch_2 In)
  (connect DefaultInputNode Clipping%20polygons ClipPaths_1 ClipPolygons)
  (connect Input%20spline Out ClipPaths_1 In)
  (connect AttributeCompareOp_0 Out ExtractAttribute_0 In)
  (connect AttributeMathsOp_0 Out ExtractAttribute_1 In)
  (connect AttributeMathsOp_1 Out ExtractAttribute_1 In)
  (connect ClipPaths_1 Out GetElementsCount_0 In)
  (connect ExtractAttribute_1 Out MergeAttributes_0 In)
  (output Out AddTags_0)
  (output Out AddTags_1)
  (output Out AddTags_3)
  (output Out AddTags_5))
